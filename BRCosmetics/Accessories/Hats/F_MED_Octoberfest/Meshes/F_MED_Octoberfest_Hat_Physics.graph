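# Physics asset for the F_MED Octoberfest hat cosmetic, defining secondary-motion physics for its left and right ponytail and hair bones.
# Two four-bone ponytail chains and two two-bone hair chains hang from the head as capsules, constrained by swing limits and very tight twist limits.
# Head, neck, upper spine, clavicle and upper-arm bodies form an upper-body collision proxy of capsules and a box, linked by constraints with free swing and twist.
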PhysicsAsset /BRCosmetics/Accessories/Hats/F_MED_Octoberfest/Meshes/F_MED_Octoberfest_Hat_Physics
bodies {
 clavicle_l [{shapeType=Capsule, length=7.5801, radius=6.9382, center=<5.2163 -1.2987 -2.8186>, rotation=<rot 90 0 0>}]
 clavicle_r [{shapeType=Capsule, length=7.5801, radius=6.9382, center=<-5.2163 1.2987 2.8186>, rotation=<rot 90 0 0>}]
 dyn_hair_l_01 [{shapeType=Capsule, length=6, radius=1, shapeName=dyn_hair_l_01_capsule, center=<3.6337 0.0001 -0.0001>, rotation=<rot 90 0 0>}]
 dyn_hair_l_02 [{shapeType=Capsule, length=6, radius=1, shapeName=dyn_hair_l_02_capsule, center=<3.6337 0.0001 -0.0001>, rotation=<rot 90 0 0>}]
 dyn_hair_r_01 [{shapeType=Capsule, length=6, radius=1, shapeName=dyn_hair_r_01_capsule, center=<-3.6337 0 0>, rotation=<rot 90 0 0>}]
 dyn_hair_r_02 [{shapeType=Capsule, length=6, radius=1, shapeName=dyn_hair_r_02_capsule, center=<-3.6337 0 0>, rotation=<rot 90 0 0>}]
 dyn_ponytail_l_01 [{shapeType=Capsule, length=2.2809, radius=1.7191, center=<0 0 3>}]
 dyn_ponytail_l_02 [{shapeType=Capsule, length=7.4273, radius=2, center=<0 0 4.814>}]
 dyn_ponytail_l_03 [{shapeType=Capsule, length=5.1086, radius=1.646, center=<0 0 3>}]
 dyn_ponytail_l_04 [{shapeType=Capsule, length=3.2918, radius=2, center=<0 0 3>}]
 dyn_ponytail_r_01 [{shapeType=Capsule, length=2.2809, radius=1.7191, center=<0 0 -3>}]
 dyn_ponytail_r_02 [{shapeType=Capsule, length=7.4273, radius=2, center=<0 0 -4.814>}]
 dyn_ponytail_r_03 [{shapeType=Capsule, length=5.1086, radius=1.646, center=<0 0 -3>}]
 dyn_ponytail_r_04 [{shapeType=Capsule, length=3.2918, radius=2, center=<0 0 -3>}]
 head [
  {shapeType=Capsule, length=2.85, radius=5.7337, center=<1.5439 3.4489 0.0001>, rotation=<rot -89.5256 -89.7443 90.2186>}
  {shapeType=Capsule, length=7.4331, radius=2.7748, center=<1.1568 -0.1738 -6.9765>, rotation=<rot -84.9996 -0.0001 10.0014>}
  {shapeType=Capsule, length=7.4331, radius=2.7748, center=<1.1568 -0.1738 6.9765>, rotation=<rot 84.9996 -0.0001 -10.0014>}
 ]
 neck_02 [{shapeType=Capsule, length=10.9843, radius=4.6791, rotation=<rot 0 -89.9999 -89.9996>}]
 spine_05 [
  {shapeType=Capsule, length=13.5683, radius=5.9859, center=<-1.6153 2.8959 -5.101>, rotation=<rot -84.993 0 -49.9992>}
  {shapeType=Capsule, length=13.5683, radius=5.9859, center=<-1.5843 2.922 5.0983>, rotation=<rot 84.993 0 49.9987>}
  {shapeType=Box, extentX=14.0671, extentY=7.4811, extentZ=18.7407, center=<2.3541 2.7122 0.0005>, rotation=<rot 0 -39.9979 0>}
 ]
 upperarm_l [{shapeType=Capsule, length=20.741, radius=4.1987, center=<11.4355 0 0>, rotation=<rot 90 0 0>}]
 upperarm_r [{shapeType=Capsule, length=20.741, radius=4.1987, center=<-11.4355 0 0>, rotation=<rot 90 0 0>}]
}
constraints [
 {bone1Name=dyn_ponytail_l_01, bone2Name=head, swing1LimitDegrees=45, swing2LimitDegrees=45, twistLimitDegrees=10}
 {bone1Name=dyn_ponytail_r_01, bone2Name=head, swing1LimitDegrees=45, swing2LimitDegrees=45, twistLimitDegrees=10}
 {bone1Name=dyn_ponytail_l_02, bone2Name=dyn_ponytail_l_01, swing1LimitDegrees=30, swing2LimitDegrees=25, twistLimitDegrees=1.5}
 {bone1Name=dyn_ponytail_l_03, bone2Name=dyn_ponytail_l_02, swing1LimitDegrees=45, swing2LimitDegrees=45, twistLimitDegrees=1.5}
 {bone1Name=dyn_ponytail_l_04, bone2Name=dyn_ponytail_l_03, swing1LimitDegrees=45, swing2LimitDegrees=45, twistLimitDegrees=1.5}
 {bone1Name=dyn_ponytail_r_02, bone2Name=dyn_ponytail_r_01, swing1LimitDegrees=30, swing2LimitDegrees=25, twistLimitDegrees=1.5}
 {bone1Name=dyn_ponytail_r_03, bone2Name=dyn_ponytail_r_02, swing1LimitDegrees=45, swing2LimitDegrees=45, twistLimitDegrees=1.5}
 {bone1Name=dyn_ponytail_r_04, bone2Name=dyn_ponytail_r_03, swing1LimitDegrees=45, swing2LimitDegrees=45, twistLimitDegrees=1.5}
 {
  bone1Name=head
  bone2Name=neck_02
  swing1LimitDegrees=45
  swing1Motion=Free
  swing2LimitDegrees=45
  swing2Motion=Free
  twistLimitDegrees=45
  twistMotion=Free
 }
 {
  bone1Name=clavicle_l
  bone2Name=spine_05
  swing1LimitDegrees=45
  swing1Motion=Free
  swing2LimitDegrees=45
  swing2Motion=Free
  twistLimitDegrees=45
  twistMotion=Free
 }
 {
  bone1Name=clavicle_r
  bone2Name=spine_05
  swing1LimitDegrees=45
  swing1Motion=Free
  swing2LimitDegrees=45
  swing2Motion=Free
  twistLimitDegrees=45
  twistMotion=Free
 }
 {
  bone1Name=neck_02
  bone2Name=spine_05
  swing1LimitDegrees=45
  swing1Motion=Free
  swing2LimitDegrees=45
  swing2Motion=Free
  twistLimitDegrees=45
  twistMotion=Free
 }
 {
  bone1Name=upperarm_r
  bone2Name=clavicle_r
  swing1LimitDegrees=45
  swing1Motion=Free
  swing2LimitDegrees=45
  swing2Motion=Free
  twistLimitDegrees=45
  twistMotion=Free
 }
 {
  bone1Name=upperarm_l
  bone2Name=clavicle_l
  swing1LimitDegrees=45
  swing1Motion=Free
  swing2LimitDegrees=45
  swing2Motion=Free
  twistLimitDegrees=45
  twistMotion=Free
 }
 {bone1Name=dyn_hair_l_01, bone2Name=head, swing1LimitDegrees=30, swing2LimitDegrees=15, twistLimitDegrees=0.1}
 {bone1Name=dyn_hair_l_02, bone2Name=dyn_hair_l_01, swing1LimitDegrees=35, swing2LimitDegrees=25, twistLimitDegrees=2}
 {bone1Name=dyn_hair_r_01, bone2Name=head, swing1LimitDegrees=30, swing2LimitDegrees=15, twistLimitDegrees=0.1}
 {bone1Name=dyn_hair_r_02, bone2Name=dyn_hair_r_01, swing1LimitDegrees=35, swing2LimitDegrees=25, twistLimitDegrees=2}
]
