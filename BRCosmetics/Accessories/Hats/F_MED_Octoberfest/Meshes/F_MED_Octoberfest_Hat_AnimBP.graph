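# Post-process style AnimBlueprint for the Octoberfest hat accessory, adding secondary motion to its feather, hair and ponytail bones.
# The copied parent pose feeds RigidBody or, when that is disabled, AnimDynamics or a front-end performance-mode Look At fallback, with head-tilt corrections.
# Shared hair and feather AnimDynamics and ponytail Look Ats follow; several other chains are unconnected.
# The EventGraph computes performance mode, head tilt and skydiving force for RigidBody.
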AnimBlueprint /BRCosmetics/Accessories/Hats/F_MED_Octoberfest/Meshes/F_MED_Octoberfest_Hat_AnimBP
target_skeleton=&/F_MED_Octoberfest_Hat_Skeleton
graphs [
 AnimGraph {nodes [
   n1 = Anim Dynamics - Bone: dyn_feather_base!(ComponentPose, ExternalForce="0,0,0") -> (Pose) {
    alpha=0.2
    angularDampingOverride=0.99
    angularSpringConstant=500
    bAngularSpring=true
    bGravityOverrideInSimSpace=true
    bOverrideAngularDamping=true
    simulationSpace=BoneRelative
    alphaBoolBlend {blendOption=Linear}
    boundBone {boneName=dyn_feather_base}
    componentAppliedLinearAccClamp=<3500 3500 3500>
    componentLinearAccScale=<1.2 1.2 1.2>
    componentLinearVelScale=<1.2 1.2 1.2>
    gravityOverride {z=-100}
    physicsBodyDefinitions [{
      sphereCollisionRadius=0
      boundBone {boneName=dyn_feather_base}
      boxExtents=<15 10 10>
      constraintSetup {angularLimitsMax=<30 5 30>, angularLimitsMin=<-20 -5 -30>, angularTarget=<1 0 0>}
      localJointOffset=<1 0 0>
     }]
    relativeSpaceBone {boneName=head}
   }
   n2 = Anim Dynamics - Bone: dyn_feather_top!(ComponentPose=@n1, ExternalForce="0,0,0") -> (Pose) {
    alpha=0.2
    angularDampingOverride=0.99
    angularSpringConstant=500
    bAngularSpring=true
    bGravityOverrideInSimSpace=true
    bOverrideAngularDamping=true
    simulationSpace=BoneRelative
    alphaBoolBlend {blendOption=Linear}
    boundBone {boneName=dyn_feather_top}
    componentAppliedLinearAccClamp=<3500 3500 3500>
    componentLinearAccScale=<1.2 1.2 1.2>
    componentLinearVelScale=<1.2 1.2 1.2>
    gravityOverride {z=-100}
    physicsBodyDefinitions [{
      sphereCollisionRadius=0
      boundBone {boneName=dyn_feather_top}
      boxExtents=<15 10 10>
      constraintSetup {angularLimitsMax=<30 5 30>, angularLimitsMin=<-30 -5 -30>, angularTarget=<1 0 0>}
      localJointOffset=<1 0 0>
     }]
    relativeSpaceBone {boneName=head}
   }
   Anim Dynamics - Bone: dyn_feather_bott!(ComponentPose=@n2, ExternalForce="0,0,0") -> (Pose) {
    alpha=0.2
    angularDampingOverride=0.99
    angularSpringConstant=500
    bAngularSpring=true
    bGravityOverrideInSimSpace=true
    bOverrideAngularDamping=true
    simulationSpace=BoneRelative
    alphaBoolBlend {blendOption=Linear}
    boundBone {boneName=dyn_feather_bott}
    componentAppliedLinearAccClamp=<3500 3500 3500>
    componentLinearAccScale=<1.2 1.2 1.2>
    componentLinearVelScale=<1.2 1.2 1.2>
    gravityOverride {z=-100}
    physicsBodyDefinitions [{
      sphereCollisionRadius=0
      boundBone {boneName=dyn_feather_bott}
      boxExtents=<15 10 10>
      constraintSetup {angularLimitsMax=<30 5 30>, angularLimitsMin=<-30 -5 -30>, angularTarget=<1 0 0>}
      localJointOffset=<1 0 0>
     }]
    relativeSpaceBone {boneName=head}
   }
   n3 = Anim Dynamics - Bone: dyn_hair_l_01(Alpha=1, ExternalForce="0,0,0") {
    angularDampingOverride=0.98
    bGravityOverrideInSimSpace=true
    bOverrideAngularDamping=true
    bUseGravityOverride=true
    alphaBoolBlend {blendOption=Linear}
    boundBone {boneName=dyn_hair_l_01}
    componentLinearAccScale=<1 1 1>
    componentLinearVelScale=<1 1 1>
    gravityOverride {z=-980}
    physicsBodyDefinitions [{
      sphereCollisionRadius=0
      boundBone {boneName=dyn_hair_l_01}
      boxExtents=<10 15 15>
      constraintSetup {angularLimitsMax=<2 25 30>, angularLimitsMin=<-2 -25 0>, angularTarget=<0 0 0>}
      localJointOffset=<2 0 0>
     }]
   }
   n4 = Anim Dynamics - Bone: dyn_hair_l_02(ComponentPose=@n3, Alpha=1, ExternalForce="0,0,0") {
    angularDampingOverride=0.98
    bGravityOverrideInSimSpace=true
    bOverrideAngularDamping=true
    bUseGravityOverride=true
    alphaBoolBlend {blendOption=Linear}
    boundBone {boneName=dyn_hair_l_02}
    componentLinearAccScale=<1 1 1>
    componentLinearVelScale=<1 1 1>
    gravityOverride {z=-980}
    physicsBodyDefinitions [{
      sphereCollisionRadius=0
      boundBone {boneName=dyn_hair_l_02}
      boxExtents=<10 15 15>
      constraintSetup {angularLimitsMax=<2 25 30>, angularLimitsMin=<-2 -25 0>, angularTarget=<0 0 0>}
      localJointOffset=<2 0 0>
     }]
   }
   n5 = Anim Dynamics - Bone: dyn_hair_r_01(ComponentPose=@n4, Alpha=1, ExternalForce="0,0,0") {
    angularDampingOverride=0.98
    bGravityOverrideInSimSpace=true
    bOverrideAngularDamping=true
    bUseGravityOverride=true
    alphaBoolBlend {blendOption=Linear}
    boundBone {boneName=dyn_hair_r_01}
    componentLinearAccScale=<1 1 1>
    componentLinearVelScale=<1 1 1>
    gravityOverride {z=-980}
    physicsBodyDefinitions [{
      sphereCollisionRadius=0
      boundBone {boneName=dyn_hair_r_01}
      boxExtents=<10 15 15>
      constraintSetup {angularLimitsMax=<2 25 30>, angularLimitsMin=<-2 -25 0>, angularTarget=<0 0 0>}
      localJointOffset=<-2 0 0>
     }]
   }
   Anim Dynamics - Bone: dyn_hair_r_02(ComponentPose=@n5, Alpha=1, ExternalForce="0,0,0") {
    angularDampingOverride=0.98
    bGravityOverrideInSimSpace=true
    bOverrideAngularDamping=true
    bUseGravityOverride=true
    alphaBoolBlend {blendOption=Linear}
    boundBone {boneName=dyn_hair_r_02}
    componentLinearAccScale=<1 1 1>
    componentLinearVelScale=<1 1 1>
    gravityOverride {z=-980}
    physicsBodyDefinitions [{
      sphereCollisionRadius=0
      boundBone {boneName=dyn_hair_r_02}
      boxExtents=<10 15 15>
      constraintSetup {angularLimitsMax=<2 25 30>, angularLimitsMin=<-2 -25 0>, angularTarget=<0 0 0>}
      localJointOffset=<-2 0 0>
     }]
   }
   n6 = Anim Dynamics - Bone: dyn_hair_extra_b(Alpha=1, ExternalForce="0,0,0") {
    angularDampingOverride=0.9999
    angularSpringConstant=100
    bAngularSpring=true
    bOverrideAngularDamping=true
    simulationSpace=RootRelative
    alphaBoolBlend {blendOption=Linear}
    boundBone {boneName=dyn_hair_extra_b}
    componentAppliedLinearAccClamp=<3500 3500 3500>
    componentLinearAccScale=<1.2 1.2 1.2>
    componentLinearVelScale=<1.2 1.2 1.2>
    physicsBodyDefinitions [{
      sphereCollisionRadius=0
      boundBone {boneName=dyn_hair_extra_b}
      boxExtents=<20 10 10>
      constraintSetup {angularLimitsMax=<1 5 15>, angularLimitsMin=<-1 -30 -15>, angularTarget=<0 0 0>}
      localJointOffset=<2 0 0>
     }]
    relativeSpaceBone {boneName=dyn_root_simspace}
   }
   Anim Dynamics - Bone: dyn_hair_extra_c(ComponentPose=@n6, Alpha=1, ExternalForce="0,0,0") {
    angularDampingOverride=0.9999
    angularSpringConstant=100
    bAngularSpring=true
    bOverrideAngularDamping=true
    simulationSpace=RootRelative
    alphaBoolBlend {blendOption=Linear}
    boundBone {boneName=dyn_hair_extra_c}
    componentAppliedLinearAccClamp=<3500 3500 3500>
    componentLinearAccScale=<1.2 1.2 1.2>
    componentLinearVelScale=<1.2 1.2 1.2>
    physicsBodyDefinitions [{
      sphereCollisionRadius=0
      boundBone {boneName=dyn_hair_extra_c}
      boxExtents=<20 10 10>
      constraintSetup {angularLimitsMax=<1 5 15>, angularLimitsMin=<-1 -30 -15>, angularTarget=<0 0 0>}
      localJointOffset=<2 0 0>
     }]
    relativeSpaceBone {boneName=dyn_root_simspace}
   }
   n7 = Anim Dynamics - Bone: dyn_hair_l_01(Alpha=1, ExternalForce="0,0,0") {
    angularDampingOverride=0.9999
    angularSpringConstant=100
    bAngularSpring=true
    bOverrideAngularDamping=true
    bUsePlanarLimit=false
    bUseSphericalLimits=true
    simulationSpace=RootRelative
    alphaBoolBlend {blendOption=Linear}
    boundBone {boneName=dyn_hair_l_01}
    componentAppliedLinearAccClamp=<3500 3500 3500>
    componentLinearAccScale=<1.2 1.2 1.2>
    componentLinearVelScale=<1.2 1.2 1.2>
    physicsBodyDefinitions [{
      sphereCollisionRadius=0
      boundBone {boneName=dyn_hair_l_01}
      boxExtents=<20 10 10>
      constraintSetup {angularLimitsMax=<1 20 15>, angularLimitsMin=<-1 -5 -10>, angularTarget=<0 0 0>}
      localJointOffset=<2 0 0>
     }]
    relativeSpaceBone {boneName=neck_01}
    sphericalLimits [{limitRadius=8, drivingBone {boneName=head}, sphereLocalOffset=<5 2 0>}]
   }
   Anim Dynamics - Bone: dyn_hair_l_02(ComponentPose=@n7, Alpha=1, ExternalForce="0,0,0") {
    angularDampingOverride=0.9999
    angularSpringConstant=100
    bAngularSpring=true
    bOverrideAngularDamping=true
    bUsePlanarLimit=false
    bUseSphericalLimits=true
    simulationSpace=RootRelative
    alphaBoolBlend {blendOption=Linear}
    boundBone {boneName=dyn_hair_l_02}
    componentAppliedLinearAccClamp=<3500 3500 3500>
    componentLinearAccScale=<1.2 1.2 1.2>
    componentLinearVelScale=<1.2 1.2 1.2>
    physicsBodyDefinitions [{
      sphereCollisionRadius=0
      boundBone {boneName=dyn_hair_l_02}
      boxExtents=<20 10 10>
      constraintSetup {angularLimitsMax=<1 20 15>, angularLimitsMin=<-1 -5 -10>, angularTarget=<0 0 0>}
      localJointOffset=<2 0 0>
     }]
    relativeSpaceBone {boneName=neck_01}
    sphericalLimits [{limitRadius=7.8, drivingBone {boneName=head}, sphereLocalOffset=<0 0 0>}]
   }
   n8 = Look At - Bone: dyn_ponytail_l_02(ComponentPose=Local To Component(LocalPose=Use cached pose 'SavedPose13880'()), Alpha=1) {
    alphaBoolBlend {blendOption=Linear}
    boneToModify {boneName=dyn_ponytail_l_02}
    lookAtLocation=<0 10 -8>
    lookAtTarget {boneReference {boneName=spine_05}}
    lookAt_Axis {axis {y=0, z=1}}
   }
   n9 = Look At - Bone: dyn_ponytail_r_02(ComponentPose=@n8, Alpha=1) {
    alphaBoolBlend {blendOption=Linear}
    boneToModify {boneName=dyn_ponytail_r_02}
    lookAtLocation=<0 10 8>
    lookAtTarget {boneReference {boneName=spine_05}}
    lookAt_Axis {axis {y=0, z=-1}}
   }
   Blend Poses by bool(
    BlendPose_0=Component To Local(ComponentPose=@n9)
    BlendPose_1=Component To Local()
    bActiveValue="integer > integer"(A=Get CurrentLOD(), B=1) {is=title K2Node_CallFunction}
   ) {blendType=Linear}
   n10 = Anim Dynamics - Bone: dyn_ponytail_l_01(ComponentPose=Local To Component(LocalPose=Use cached pose 'SavedPose13880'()), Alpha=1, ExternalForce="0,0,0") {
    angularDampingOverride=0.9999
    angularSpringConstant=500
    bAngularSpring=true
    bGravityOverrideInSimSpace=true
    bOverrideAngularDamping=true
    simulationSpace=BoneRelative
    alphaBoolBlend {blendOption=Linear}
    boundBone {boneName=dyn_ponytail_l_01}
    componentAppliedLinearAccClamp=<3500 3500 3500>
    componentLinearAccScale=<1.2 1.2 1.2>
    componentLinearVelScale=<1.2 1.2 1.2>
    physicsBodyDefinitions [{
      sphereCollisionRadius=0
      boundBone {boneName=dyn_ponytail_l_01}
      boxExtents=<20 20 15>
      constraintSetup {twistAxis=AxisZ, angularLimitsMax=<1 5 15>, angularLimitsMin=<-1 -45 -15>, angularTarget=<0 0 0>}
      localJointOffset=<0 0 3>
     }]
    relativeSpaceBone {boneName=root}
   }
   n11 = Anim Dynamics - Bone: dyn_ponytail_l_02(ComponentPose=@n10, Alpha=1, ExternalForce="0,0,0") {
    angularDampingOverride=0.9999
    angularSpringConstant=500
    bAngularSpring=true
    bGravityOverrideInSimSpace=true
    bOverrideAngularDamping=true
    simulationSpace=BoneRelative
    alphaBoolBlend {blendOption=Linear}
    boundBone {boneName=dyn_ponytail_l_02}
    componentAppliedLinearAccClamp=<3500 3500 3500>
    componentLinearAccScale=<1.2 1.2 1.2>
    componentLinearVelScale=<1.2 1.2 1.2>
    physicsBodyDefinitions [{
      sphereCollisionRadius=0
      boundBone {boneName=dyn_ponytail_l_02}
      boxExtents=<20 20 15>
      constraintSetup {twistAxis=AxisZ, angularLimitsMax=<1 5 15>, angularLimitsMin=<-1 -45 -15>, angularTarget=<0 0 0>}
      localJointOffset=<0 0 3>
     }]
    relativeSpaceBone {boneName=root}
   }
   n12 = Anim Dynamics - Bone: dyn_ponytail_l_03(ComponentPose=@n11, Alpha=1, ExternalForce="0,0,0") {
    angularDampingOverride=0.9999
    angularSpringConstant=500
    bAngularSpring=true
    bGravityOverrideInSimSpace=true
    bOverrideAngularDamping=true
    bUseSphericalLimits=true
    simulationSpace=BoneRelative
    alphaBoolBlend {blendOption=Linear}
    boundBone {boneName=dyn_ponytail_l_03}
    componentAppliedLinearAccClamp=<3500 3500 3500>
    componentLinearAccScale=<1.2 1.2 1.2>
    componentLinearVelScale=<1.2 1.2 1.2>
    physicsBodyDefinitions [{
      sphereCollisionRadius=0
      boundBone {boneName=dyn_ponytail_l_03}
      boxExtents=<20 20 15>
      constraintSetup {twistAxis=AxisZ, angularLimitsMax=<1 5 15>, angularLimitsMin=<-1 -45 -15>, angularTarget=<0 0 0>}
      localJointOffset=<0 0 3>
     }]
    relativeSpaceBone {boneName=root}
   }
   n13 = Anim Dynamics - Bone: dyn_ponytail_l_04(ComponentPose=@n12, Alpha=1, ExternalForce="0,0,0") {
    angularDampingOverride=0.9999
    angularSpringConstant=500
    bAngularSpring=true
    bGravityOverrideInSimSpace=true
    bOverrideAngularDamping=true
    bUseSphericalLimits=true
    simulationSpace=BoneRelative
    alphaBoolBlend {blendOption=Linear}
    boundBone {boneName=dyn_ponytail_l_04}
    componentAppliedLinearAccClamp=<3500 3500 3500>
    componentLinearAccScale=<1.2 1.2 1.2>
    componentLinearVelScale=<1.2 1.2 1.2>
    physicsBodyDefinitions [{
      sphereCollisionRadius=0
      boundBone {boneName=dyn_ponytail_l_04}
      boxExtents=<20 20 15>
      constraintSetup {twistAxis=AxisZ, angularLimitsMax=<1 5 15>, angularLimitsMin=<-1 -45 -15>, angularTarget=<0 0 0>}
      localJointOffset=<0 0 3>
     }]
    relativeSpaceBone {boneName=root}
   }
   n14 = Anim Dynamics - Bone: dyn_ponytail_r_01(ComponentPose=@n13, Alpha=1, ExternalForce="0,0,0") {
    angularDampingOverride=0.9999
    angularSpringConstant=500
    bAngularSpring=true
    bGravityOverrideInSimSpace=true
    bOverrideAngularDamping=true
    simulationSpace=BoneRelative
    alphaBoolBlend {blendOption=Linear}
    boundBone {boneName=dyn_ponytail_r_01}
    componentAppliedLinearAccClamp=<3500 3500 3500>
    componentLinearAccScale=<1.2 1.2 1.2>
    componentLinearVelScale=<1.2 1.2 1.2>
    physicsBodyDefinitions [{
      sphereCollisionRadius=0
      boundBone {boneName=dyn_ponytail_r_01}
      boxExtents=<20 20 15>
      constraintSetup {twistAxis=AxisZ, angularLimitsMax=<1 5 15>, angularLimitsMin=<-1 -45 -15>, angularTarget=<0 0 0>}
      localJointOffset=<0 0 -3>
     }]
    relativeSpaceBone {boneName=root}
   }
   n15 = Anim Dynamics - Bone: dyn_ponytail_r_02(ComponentPose=@n14, Alpha=1, ExternalForce="0,0,0") {
    angularDampingOverride=0.9999
    angularSpringConstant=500
    bAngularSpring=true
    bGravityOverrideInSimSpace=true
    bOverrideAngularDamping=true
    simulationSpace=BoneRelative
    alphaBoolBlend {blendOption=Linear}
    boundBone {boneName=dyn_ponytail_r_02}
    componentAppliedLinearAccClamp=<3500 3500 3500>
    componentLinearAccScale=<1.2 1.2 1.2>
    componentLinearVelScale=<1.2 1.2 1.2>
    physicsBodyDefinitions [{
      sphereCollisionRadius=0
      boundBone {boneName=dyn_ponytail_r_02}
      boxExtents=<20 20 15>
      constraintSetup {twistAxis=AxisZ, angularLimitsMax=<1 5 15>, angularLimitsMin=<-1 -45 -15>, angularTarget=<0 0 0>}
      localJointOffset=<0 0 -3>
     }]
    relativeSpaceBone {boneName=root}
   }
   n16 = Anim Dynamics - Bone: dyn_ponytail_r_03(ComponentPose=@n15, Alpha=1, ExternalForce="0,0,0") {
    angularDampingOverride=0.9999
    angularSpringConstant=500
    bAngularSpring=true
    bGravityOverrideInSimSpace=true
    bOverrideAngularDamping=true
    simulationSpace=BoneRelative
    alphaBoolBlend {blendOption=Linear}
    boundBone {boneName=dyn_ponytail_r_03}
    componentAppliedLinearAccClamp=<3500 3500 3500>
    componentLinearAccScale=<1.2 1.2 1.2>
    componentLinearVelScale=<1.2 1.2 1.2>
    physicsBodyDefinitions [{
      sphereCollisionRadius=0
      boundBone {boneName=dyn_ponytail_r_03}
      boxExtents=<20 20 15>
      constraintSetup {twistAxis=AxisZ, angularLimitsMax=<1 1 15>, angularLimitsMin=<-1 -45 -15>, angularTarget=<0 0 0>}
      localJointOffset=<0 0 -3>
     }]
    relativeSpaceBone {boneName=root}
   }
   n17 = Anim Dynamics - Bone: dyn_ponytail_r_04(ComponentPose=@n16, Alpha=1, ExternalForce="0,0,0") {
    angularDampingOverride=0.9999
    angularSpringConstant=500
    bAngularSpring=true
    bGravityOverrideInSimSpace=true
    bOverrideAngularDamping=true
    simulationSpace=BoneRelative
    alphaBoolBlend {blendOption=Linear}
    boundBone {boneName=dyn_ponytail_r_04}
    componentAppliedLinearAccClamp=<3500 3500 3500>
    componentLinearAccScale=<1.2 1.2 1.2>
    componentLinearVelScale=<1.2 1.2 1.2>
    physicsBodyDefinitions [{
      sphereCollisionRadius=0
      boundBone {boneName=dyn_ponytail_r_04}
      boxExtents=<20 20 15>
      constraintSetup {twistAxis=AxisZ, angularLimitsMax=<1 1 15>, angularLimitsMin=<-1 -45 -15>, angularTarget=<0 0 0>}
      localJointOffset=<0 0 -3>
     }]
    relativeSpaceBone {boneName=root}
   }
   n18 = "Transform (Modify) Bone - Bone: dyn_ponytail_l_01"!(ComponentPose=@n17, Alpha=1, Rotation="-18,-30,10") -> (Pose) {
    rotationMode=BMM_Replace
    rotationSpace=BCS_BoneSpace
    translationMode=BMM_Replace
    translationSpace=BCS_BoneSpace
    alphaBoolBlend {blendOption=Linear}
    alphaScaleBiasClamp {bInterpResult=true, interpSpeedDecreasing=5, interpSpeedIncreasing=7}
    boneToModify {boneName=dyn_ponytail_l_01}
    rotation=<rot -18 -30 10>
   }
   n19 = "Transform (Modify) Bone - Bone: dyn_ponytail_r_01"!(ComponentPose=@n18, Alpha=1, Rotation="-15,-40,11") -> (Pose) {
    rotationMode=BMM_Replace
    rotationSpace=BCS_BoneSpace
    translationMode=BMM_Replace
    translationSpace=BCS_BoneSpace
    alphaBoolBlend {blendOption=Linear}
    alphaScaleBiasClamp {bInterpResult=true, interpSpeedDecreasing=5, interpSpeedIncreasing=7}
    boneToModify {boneName=dyn_ponytail_r_01}
    rotation=<rot -15 -40 11>
   }
   n20 = Get Head Up() -> (Head Up:double)
   n21 = "Transform (Modify) Bone - Bone: dyn_ponytail_l_01"(ComponentPose=@n19, Alpha=@n20, Rotation="-15.000,-0.033,5.053") {
    rotationMode=BMM_Replace
    rotationSpace=BCS_BoneSpace
    translationMode=BMM_Replace
    translationSpace=BCS_BoneSpace
    alphaBoolBlend {blendOption=Linear}
    alphaScaleBiasClamp {bInterpResult=true, interpSpeedDecreasing=5, interpSpeedIncreasing=7}
    boneToModify {boneName=dyn_ponytail_l_01}
    rotation=<rot -15 -0.033 5.053>
   }
   n22 = "Transform (Modify) Bone - Bone: dyn_ponytail_r_01"(ComponentPose=@n21, Alpha=@n20, Rotation="-14.971,-0.002,9.927") {
    rotationMode=BMM_Replace
    rotationSpace=BCS_BoneSpace
    translationMode=BMM_Replace
    translationSpace=BCS_BoneSpace
    alphaBoolBlend {blendOption=Linear}
    alphaScaleBiasClamp {bInterpResult=true, interpSpeedDecreasing=5, interpSpeedIncreasing=7}
    boneToModify {boneName=dyn_ponytail_r_01}
    rotation=<rot -14.971 -0.002 9.927>
   }
   n23 = Look At - Bone: dyn_ponytail_l_01(ComponentPose=Local To Component(LocalPose=Use cached pose 'SavedPose13880'()), Alpha=1) {
    boneToModify {boneName=dyn_ponytail_l_01}
    lookAtLocation {x=-3, z=-5}
    lookAtTarget {boneReference {boneName=spine_05}}
    lookAt_Axis {axis {y=0, z=1}}
    lookUp_Axis {axis {x=0, z=1}}
   }
   n24 = Look At - Bone: dyn_ponytail_r_01(ComponentPose=@n23, Alpha=1) {
    boneToModify {boneName=dyn_ponytail_r_01}
    lookAtLocation=<0 4 5>
    lookAtTarget {boneReference {boneName=spine_05}}
    lookAt_Axis {axis {y=0, z=-1}}
    lookUp_Axis {axis {x=0, z=1}}
   }
   n25 = Get Head Up()
   n26 = "Transform (Modify) Bone - Bone: dyn_ponytail_l_01"(ComponentPose=@n24, Alpha=@n25, Rotation="-10,0,5") {
    rotationMode=BMM_Replace
    rotationSpace=BCS_BoneSpace
    translationMode=BMM_Replace
    translationSpace=BCS_BoneSpace
    alphaBoolBlend {blendOption=Linear}
    alphaScaleBiasClamp {bInterpResult=true, interpSpeedDecreasing=5, interpSpeedIncreasing=7}
    boneToModify {boneName=dyn_ponytail_l_01}
    rotation {pitch=-10, roll=5}
   }
   n27 = "Transform (Modify) Bone - Bone: dyn_ponytail_r_01"(ComponentPose=@n26, Alpha=@n25, Rotation="-10,0,0") {
    rotationMode=BMM_Replace
    rotationSpace=BCS_BoneSpace
    translationMode=BMM_Replace
    translationSpace=BCS_BoneSpace
    alphaBoolBlend {blendOption=Linear}
    alphaScaleBiasClamp {bInterpResult=true, interpSpeedDecreasing=5, interpSpeedIncreasing=7}
    boneToModify {boneName=dyn_ponytail_r_01}
    rotation {pitch=-10}
   }
   n28 = Blend Poses by bool(
    BlendPose_0=Component To Local(ComponentPose=@n27)
    BlendPose_1=Component To Local(ComponentPose=@n22)
    bActiveValue=Get IsPerformanceModeInFE() -> (IsPerformanceModeInFE:bool)
   )
   n29 = RigidBody(
    ComponentPose=Local To Component(LocalPose=Use cached pose 'SavedPose13880'())
    ExternalForce=Get skydiving vector force() -> (skydiving vector force:Vector)
   ) {
    bOverrideWorldGravity=true
    overridePhysicsAsset=&/F_MED_Octoberfest_Hat_Physics
    simulationSpace=BaseBoneSpace
    alphaBoolBlend {blendOption=Linear}
    baseBoneRef {boneName=spine_05}
    componentAppliedLinearAccClamp=<3000 3000 3000>
    componentLinearAccScale=<0.3 0.2 0.2>
    componentLinearVelScale=<0.3 0.2 0.2>
    overrideWorldGravity {z=-980}
   }
   n30 = Component To Local(ComponentPose=@n29)
   n31 = Get bRigidBodyEnabled()
   n32 = Anim Dynamics - Bone: dyn_hair_l_01(
    ComponentPose=Local To Component(LocalPose=Blend Poses by bool(BlendPose_0=@n30, BlendPose_1=@n28, bActiveValue=@n31) {blendType=Linear})
    Alpha=1
    ExternalForce="0,0,0"
   ) {
    angularDampingOverride=0.9999
    bGravityOverrideInSimSpace=true
    bOverrideAngularDamping=true
    bUseSphericalLimits=true
    gravityScale=0.35
    simulationSpace=RootRelative
    alphaBoolBlend {blendOption=Linear}
    boundBone {boneName=dyn_hair_l_01}
    componentLinearAccScale {z=0.5}
    componentLinearVelScale {z=0.5}
    physicsBodyDefinitions [{
      sphereCollisionRadius=0
      boundBone {boneName=dyn_hair_l_01}
      boxExtents=<5 20 20>
      constraintSetup {angularLimitsMax=<0 20 10>, angularLimitsMin=<0 -5 -10>, angularTarget=<0 0 0>}
      localJointOffset=<5 0 0>
     }]
    sphericalLimits [{limitRadius=8.5, drivingBone {boneName=head}, sphereLocalOffset=<5 2 0>}]
   }
   n33 = Anim Dynamics - Bone: dyn_hair_l_02(ComponentPose=@n32, Alpha=1, ExternalForce="0,0,0") {
    angularDampingOverride=1
    bGravityOverrideInSimSpace=true
    bOverrideAngularDamping=true
    bUseSphericalLimits=true
    gravityScale=0.35
    simulationSpace=RootRelative
    alphaBoolBlend {blendOption=Linear}
    boundBone {boneName=dyn_hair_l_02}
    componentLinearAccScale {z=0.5}
    componentLinearVelScale {z=0.5}
    physicsBodyDefinitions [{
      sphereCollisionRadius=0
      boundBone {boneName=dyn_hair_l_02}
      boxExtents=<5 20 20>
      constraintSetup {angularLimitsMax=<0 30 10>, angularLimitsMin=<0 0 -10>, angularTarget=<0 0 0>}
      localJointOffset=<5 0 0>
     }]
    sphericalLimits [{limitRadius=8, drivingBone {boneName=head}, sphereLocalOffset=<0 0 0>}]
   }
   n34 = Anim Dynamics - Bone: dyn_hair_r_01(ComponentPose=@n33, Alpha=1, ExternalForce="0,0,0") {
    angularDampingOverride=0.9999
    bGravityOverrideInSimSpace=true
    bOverrideAngularDamping=true
    bUseSphericalLimits=true
    gravityScale=0.35
    simulationSpace=RootRelative
    alphaBoolBlend {blendOption=Linear}
    boundBone {boneName=dyn_hair_r_01}
    componentLinearAccScale {z=0.5}
    componentLinearVelScale {z=0.5}
    physicsBodyDefinitions [{
      sphereCollisionRadius=0
      boundBone {boneName=dyn_hair_r_01}
      boxExtents=<5 20 20>
      constraintSetup {angularLimitsMax=<0 30 10>, angularLimitsMin=<0 -15 -10>, angularTarget=<0 0 0>}
      localJointOffset=<-5 0 0>
     }]
    sphericalLimits [{limitRadius=8.5, drivingBone {boneName=head}, sphereLocalOffset=<5 2 0>}]
   }
   n35 = Anim Dynamics - Bone: dyn_hair_r_02(ComponentPose=@n34, Alpha=1, ExternalForce="0,0,0") {
    angularDampingOverride=1
    bGravityOverrideInSimSpace=true
    bOverrideAngularDamping=true
    bUseSphericalLimits=true
    gravityScale=0.35
    simulationSpace=RootRelative
    alphaBoolBlend {blendOption=Linear}
    boundBone {boneName=dyn_hair_r_02}
    componentLinearAccScale {z=0.5}
    componentLinearVelScale {z=0.5}
    physicsBodyDefinitions [{
      sphereCollisionRadius=0
      boundBone {boneName=dyn_hair_r_02}
      boxExtents=<5 20 20>
      constraintSetup {angularLimitsMax=<0 30 10>, angularLimitsMin=<0 0 -10>, angularTarget=<0 0 0>}
      localJointOffset=<-5 0 0>
     }]
    sphericalLimits [{limitRadius=8, drivingBone {boneName=head}, sphereLocalOffset=<0 0 0>}]
   }
   n36 = "Constraint - dyn_ponytail_r_02 (P)"(ComponentPose=@n35, Alpha=1, ConstraintWeights_0=1) {
    alpha=0
    alphaBoolBlend {blendOption=Linear}
    alphaScaleBiasClamp {bInterpResult=true, interpSpeedDecreasing=5, interpSpeedIncreasing=5}
    boneToModify {boneName=dyn_ponytail_r_02}
    constraintSetup [{targetBone {boneName=neck_01}}]
    constraintWeights [1]
   }
   n37 = "Transform (Modify) Bone - Bone: dyn_ponytail_r_01"(ComponentPose=@n36, Alpha=1, Rotation="-0.000,0.000,-15.000") {
    rotationMode=BMM_Replace
    rotationSpace=BCS_BoneSpace
    translationMode=BMM_Replace
    translationSpace=BCS_BoneSpace
    alphaBoolBlend {blendOption=Linear}
    alphaScaleBiasClamp {bInterpResult=true, interpSpeedDecreasing=5, interpSpeedIncreasing=7}
    boneToModify {boneName=dyn_ponytail_r_01}
    rotation=<rot 0 0 -15>
   }
   n38 = Anim Dynamics - Bone: dyn_feather_base!(ComponentPose=@n37, Alpha:float=0.5, ExternalForce="0,0,0") -> (Pose) {
    angularDampingOverride=0.9999
    angularSpringConstant=100
    bAngularSpring=true
    bOverrideAngularDamping=true
    simulationSpace=RootRelative
    alphaBoolBlend {blendOption=Linear}
    boundBone {boneName=dyn_feather_base}
    componentAppliedLinearAccClamp=<3500 3500 3500>
    componentLinearAccScale=<1.2 1.2 1.2>
    componentLinearVelScale=<1.2 1.2 1.2>
    physicsBodyDefinitions [{
      sphereCollisionRadius=0
      boundBone {boneName=dyn_feather_base}
      boxExtents=<20 10 10>
      constraintSetup {angularLimitsMax=<1 10 5>, angularLimitsMin=<-1 -10 -10>, angularTarget=<0 0 0>}
      localJointOffset=<-5 0 0>
     }]
    relativeSpaceBone {boneName=neck_01}
   }
   n39 = Anim Dynamics - Bone: dyn_feather_bott!(ComponentPose=@n38, Alpha:float=0.5, ExternalForce="0,0,0") -> (Pose) {
    angularDampingOverride=0.9999
    angularSpringConstant=100
    bAngularSpring=true
    bOverrideAngularDamping=true
    simulationSpace=RootRelative
    alphaBoolBlend {blendOption=Linear}
    boundBone {boneName=dyn_feather_bott}
    componentAppliedLinearAccClamp=<3500 3500 3500>
    componentLinearAccScale=<1.2 1.2 1.2>
    componentLinearVelScale=<1.2 1.2 1.2>
    physicsBodyDefinitions [{
      sphereCollisionRadius=0
      boundBone {boneName=dyn_feather_bott}
      boxExtents=<20 10 10>
      constraintSetup {angularLimitsMax=<1 20 2>, angularLimitsMin=<-1 -5 -15>, angularTarget=<0 0 0>}
      localJointOffset=<-5 0 0>
     }]
    relativeSpaceBone {boneName=neck_01}
   }
   n40 = Anim Dynamics - Bone: dyn_feather_top!(ComponentPose=@n39, Alpha:float=0.45, ExternalForce="0,0,0") -> (Pose) {
    angularDampingOverride=0.9999
    angularSpringConstant=100
    bAngularSpring=true
    bOverrideAngularDamping=true
    simulationSpace=RootRelative
    alphaBoolBlend {blendOption=Linear}
    boundBone {boneName=dyn_feather_top}
    componentAppliedLinearAccClamp=<3500 3500 3500>
    componentLinearAccScale=<1.2 1.2 1.2>
    componentLinearVelScale=<1.2 1.2 1.2>
    physicsBodyDefinitions [{
      sphereCollisionRadius=0
      boundBone {boneName=dyn_feather_top}
      boxExtents=<20 10 10>
      constraintSetup {angularLimitsMax=<1 20 2>, angularLimitsMin=<-1 -5 -15>, angularTarget=<0 0 0>}
      localJointOffset=<5 0 0>
     }]
    relativeSpaceBone {boneName=neck_01}
   }
   n41 = Look At - Bone: dyn_ponytail_l_02(ComponentPose=@n40, Alpha=1) {
    alphaBoolBlend {blendOption=Linear}
    boneToModify {boneName=dyn_ponytail_l_02}
    lookAtLocation=<0 10 -8>
    lookAtTarget {boneReference {boneName=spine_05}}
    lookAt_Axis {axis {y=0, z=1}}
   }
   n42 = Look At - Bone: dyn_ponytail_r_02(ComponentPose=@n41, Alpha=1) {
    alphaBoolBlend {blendOption=Linear}
    boneToModify {boneName=dyn_ponytail_r_02}
    lookAtLocation=<0 10 8>
    lookAtTarget {boneReference {boneName=spine_05}}
    lookAt_Axis {axis {y=0, z=-1}}
   }
   Output Pose(Result=Component To Local(ComponentPose=@n42)) {layerGroup=None}
   Save cached pose 'SavedPose13880'(Pose=Copy Pose From Mesh() {bUseAttachedParent=true}) {cacheName=SavedPose13880}
   SelectVector(A="0, 0, 0", B="0, 0, 0", bPickA=Get bIsSkyDiving())
  ]}
 EventGraph {nodes [
   Event BlueprintUpdateAnimation()
   > n1 = Sequence()
   > Set IsPerformanceModeInFE(IsPerformanceModeInFE:bool=AND Boolean(A=Get bIsInFrontEnd(), B=isPcPerfMode!() -> (ReturnValue:bool))) -> (Output_Get:bool)
   n2 = Get CurrentLOD()
   Set lod23(
    execute=@n1.then_1
    lod23:double="To Float (Boolean)"(InBool="integer < integer"!(A:int=@n2, B:int=2) -> (ReturnValue:bool) {is=title K2Node_CallFunction})
   ) -> (Output_Get:double)
   > n3 = Branch(Condition=Get bCanPlayCustomAnimations() -> (bCanPlayCustomAnimations:bool))
   n4 = Branch(execute=[@n3.else, @n3.then], Condition=Get bIsSkyDiving())
   n5 = Branch(execute=[@n4.else, @n4.then], Condition=Get bIsSkydiveFloating())
   n6 = Get Direction Between Sockets!(
    Component:Skeletal Mesh Component Object Reference=GetOwnerSkeletalMeshComponent()
    SocketOrBoneNameFrom:Name=head
    SocketOrBoneNameTo:Name=pelvis
   ) -> (ReturnValue:Vector) {is=title K2Node_CallFunction}
   > Set skydiving vector force(skydiving vector force:Vector=vector + vector(
     A:Vector=vector * vector(A:Vector=@n6, B:Vector="75,75,75") -> (ReturnValue:Vector) {is=title K2Node_CommutativeAssociativeBinaryOperator}
     B:Vector="0, 0,2000"
    ) -> (ReturnValue:Vector) {is=title K2Node_CommutativeAssociativeBinaryOperator}) -> (Output_Get:Vector)
   n7 = Branch(execute=@n5.else, Condition=Get bIsSkydiveDiveMode())
   > Set skydiving vector force(skydiving vector force=vector + vector(A=vector * vector(A=@n6, B="180,180,200"), B="0, 0,2000"))
   Set skydiving vector force(execute=@n7.else, skydiving vector force=vector * vector(A=@n6, B="-15,-15,-15"))
   n8 = GetOwnerSkeletalMeshComponent()
   n9 = BreakRotator(InRot=BreakTransform(InTransform=GetDeltaTransformFromRefPose(self=@n8, BoneName=head, BaseName=spine_05)).Rotation)
   > n10 = Set Head Up(Head Up:double=MapRangeClamped(Value=@n9.Yaw, InRangeA=-3, InRangeB=-40)) -> (Output_Get:double)
   > PrintString!(execute, InString="To String (Float)"(InDouble=@n10), bPrintToScreen, bPrintToLog, TextColor, Duration, Key) -> (then)
   > Set Head Down(Head Down:double=MapRangeClamped(Value=@n9.Yaw, InRangeA=3, InRangeB=20)) -> (Output_Get:double)
   > Set Head Left(Head Left:double=MapRangeClamped(Value=@n9.Pitch, InRangeA=1, InRangeB=-1)) -> (Output_Get:double)
   TryGetPawnOwner()
  ]}
]
variables [
 Head Down {type=number}
 Head Left {type=number}
 Head Up {type=number}
 IsPerformanceModeInFE {type=boolean}
 lod23 {type=number}
 skydiving vector force {type=Vector}
]
